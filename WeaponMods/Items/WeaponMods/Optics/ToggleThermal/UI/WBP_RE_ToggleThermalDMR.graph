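# Reticle extension widget for the thermal DMR optic, showing a toggle icon and scope-mode keybind prompt.
# On construct it finds the weapon's TargetModeSelector component, syncs visibility and icon material via ResetState, and binds mode changes until destruct.
# Mode changes play Anim_ToggleOn forward or in reverse, animating the icon's inner-circle alpha and green tint.
# Aiming down sights shows the prompt after a short timer; leaving hides it and cancels the timer.
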WidgetBlueprint /WeaponMods/Items/WeaponMods/Optics/ToggleThermal/UI/WBP_RE_ToggleThermalDMR
animations [Anim_ToggleOn {
  binding_count=1
  length_seconds=0.2006
  object=&self:Anim_ToggleOn
  sequence {bindings [Image_Toggle {
     id=#1
     possessed_class=Image
     tracks [
      {
       type=MovieSceneWidgetMaterialTrack
       sections [{type=MovieSceneParameterSection, channels [InnerCircle_Alpha {type=MovieSceneScriptingFloatChannel, keys [[0,0],[4.0123,1]]}]}]
      }
      {
       type=MovieSceneColorTrack
       display_name=Color and Opacity
       sections [{
         type=MovieSceneColorSection
         end_frame=4
         start_frame=0
         channels [
          Color.R {type=MovieSceneScriptingFloatChannel, keys [[0,1],[4.0123,0]]}
          Color.G {type=MovieSceneScriptingFloatChannel, const [1,2]}
          Color.B {type=MovieSceneScriptingFloatChannel, keys [[0,1],[4.0123,0.25]]}
          Color.A {type=MovieSceneScriptingFloatChannel, const [1,2]}
         ]
        }]
      }
     ]
    }]}
 }]
graphs [
 EventGraph {nodes [
   n1 = OnTargetModeChanged!() -> (OutputDelegate, then, TargetMode:int) {is=title K2Node_CustomEvent}
   > n2 = Branch(Condition="Not Equal (Integer)"(A=@n1.TargetMode))
   n3 = Get Anim_ToggleOn() -> (Anim_ToggleOn:Widget Animation Object Reference)
   > PlayAnimationForward(InAnimation=@n3)
   PlayAnimationReverse(execute=@n2.else, InAnimation=@n3)
   n4 = ShowTogglePrompt() {is=title K2Node_CustomEvent}
   n5 = Get HorizontalBox_TogglePrompt() -> (HorizontalBox_TogglePrompt:Horizontal Box Object Reference)
   > SetVisibility(self=@n5, InVisibility=SelfHitTestInvisible)
   n6 = Event OnTargetingChanged!() -> (OutputDelegate, then, bIsADS:bool) {is=title K2Node_Event}
   > n7 = Branch(Condition=@n6.bIsADS)
   > n8 = Set Timer by Event(Delegate=@n4, Time=0.2, bLooping=false)
   > Set TogglePromptShowDelay(TogglePromptShowDelay:Timer Handle Structure=@n8) -> (Output_Get:Timer Handle Structure)
   SetVisibility(execute=@n7.else, self=@n5, InVisibility=Hidden)
   > Clear and Invalidate Timer by Handle(Handle=Get TogglePromptShowDelay() -> (TogglePromptShowDelay:Timer Handle Structure))
   Event Construct()
   > Parent: Construct()
   > n9 = Set EquippedWeapon(EquippedWeapon:Fort Weapon Object Reference=GetEquippedWeapon!(self:Fort Weapon Reticle Extension Widget Base Object Reference) -> (ReturnValue:Fort Weapon Object Reference)) -> (Output_Get:Fort Weapon Object Reference)
   > n10 = Set TargetModeSelectorComponent(TargetModeSelectorComponent:Fort Weapon Component Target Mode Selector Object Reference=GetComponentByClass!(self:Actor Object Reference=@n9, ComponentClass:Actor Component Class Reference=/Script/FortniteGame.FortWeaponComponent_TargetModeSelector) -> (ReturnValue:Fort Weapon Component Target Mode Selector Object Reference)) -> (Output_Get:Fort Weapon Component Target Mode Selector Object Reference)
   > ResetState()
   > Is Valid(InputObject=@n10)
   > Bind Event to On Target Mode Changed(self:Fort Weapon Component Target Mode Selector Object Reference=@n10, Delegate=Create Event()) {is=title K2Node_AddDelegate}
   Event Destruct()
   > Parent: Destruct()
   n11 = Get TargetModeSelectorComponent() -> (TargetModeSelectorComponent:Fort Weapon Component Target Mode Selector Object Reference)
   > Is Valid(InputObject=@n11)
   > Unbind Event from On Target Mode Changed(self:Fort Weapon Component Target Mode Selector Object Reference=@n11, Delegate=Create Event()) {is=title K2Node_RemoveDelegate}
  ]}
 ResetState {nodes [
   ResetState() {is=title K2Node_FunctionEntry}
   > SetVisibility(
    self=Get HorizontalBox_TogglePrompt()
    InVisibility=Select(
     Option 0=Hidden
     Option 1=SelfHitTestInvisible
     Index=GetIsADS!(self:Fort Weapon Reticle Extension Widget Base Object Reference) -> (ReturnValue:bool)
    )
   )
   n1 = Get TargetModeSelectorComponent()
   > Is Valid(InputObject=@n1)
   > n2 = GetDynamicMaterial(self=Get Image_Toggle() -> (Image_Toggle:Image Object Reference))
   n3 = Get TargetMode!(self:Fort Weapon Component Target Mode Selector Object Reference=@n1) -> (TargetMode:int=0)
   > SetScalarParameterValue(
    self=@n2
    ParameterName=InnerCircle_Alpha
    Value=Select(Option 0:float=0, Option 1:float=1, Index="Not Equal (Integer)"(A=@n3)) -> (ReturnValue:float)
   )
  ]}
]
variables [
 EquippedWeapon {type=FortWeapon, default=null}
 TargetModeSelectorComponent {type=FortWeaponComponent_TargetModeSelector, default=null}
 TogglePromptShowDelay {type=TimerHandle}
]
widgets {
 info {
  inheritedWidgetCount=1
  parentClass=/Script/FortniteUI.FortWeaponReticleExtensionWidgetBase
  rootWidgetClass=&SizeBox
  widgetCount=7
 }
 widgets [
  SizeBox_1 {
   widgetClassPath=&SizeBox
   widget=SizeBox_1
   properties {bOverride_HeightOverride=true, bOverride_WidthOverride=true, heightOverride=375, renderOpacity=0.85, widthOverride=475}
  }
  HorizontalBox_TogglePrompt {
   widgetClassPath=&HorizontalBox
   bIsVariable=true
   parent=SizeBox_1
   widget=HorizontalBox_TogglePrompt
   properties {renderTransform {angle=0, scale=<1 1>, shear=<0 0>, translation=<125 120>}}
   slot {class=SizeBoxSlot, properties {horizontalAlignment=HAlign_Center, verticalAlignment=VAlign_Center}}
  }
  Image_Toggle {
   widgetClassPath=&Image
   bIsVariable=true
   parent=HorizontalBox_TogglePrompt
   widget=Image_Toggle
   properties {
    visibility=SelfHitTestInvisible
    brush {
     bIsDynamicallyLoaded=false
     drawAs=Image
     imageType=FullColor
     mirroring=NoMirror
     resourceName=None
     resourceObject=&/MI_UI_ToggleWeapon_Thermal
     tiling=NoTile
     imageSize=<25 25>
     margin=<ltrb 0 0 0 0>
     outlineSettings {
      bUseBrushTransparency=false
      roundingType=HalfHeightRadius
      width=0
      color {colorUseRule=UseColor_Specified, specifiedColor=<rgba 0 0 0 0>}
      cornerRadii=<0 0 0 0>
     }
     tintColor {colorUseRule=UseColor_Specified, specifiedColor=<rgba 1 1 1 1>}
     uVRegion {bIsValid=false, max=<0 0>, min=<0 0>}
    }
    colorAndOpacity=<rgba 0 1 0.25 1>
   }
   slot {class=HorizontalBoxSlot, properties {horizontalAlignment=HAlign_Center, verticalAlignment=VAlign_Center}}
  }
  Spacer_109 {
   widgetClassPath=&Spacer
   parent=HorizontalBox_TogglePrompt
   widget=Spacer_109
   properties {size=<2 1>}
   slot {class=HorizontalBoxSlot}
  }
  SizeBox_0 {
   widgetClassPath=&SizeBox
   parent=HorizontalBox_TogglePrompt
   widget=SizeBox_0
   properties {bOverride_HeightOverride=true, heightOverride=32}
   slot {class=HorizontalBoxSlot, properties {horizontalAlignment=HAlign_Left, verticalAlignment=VAlign_Center}}
  }
  Keybind_Toggle {
   widgetClassPath=/FortUILibrary/Input/KeybindWidget
   parent=SizeBox_0
   widget=Keybind_Toggle
   properties {
    boundInputAction=/ToggleScopeModeInput/Input/IA_ToggleScopeMode
    forcedHoldKeybindStatus=NeverShowHold
    boundActionFallbacks [UseOrReload]
    renderTransform {angle=0, scale=<0.8 0.8>, shear=<0 0>, translation=<0 0>}
   }
   slot {class=SizeBoxSlot, properties {horizontalAlignment=HAlign_Left, verticalAlignment=VAlign_Center}}
  }
  TutorialHighlight {widgetClassPath=&TutorialHighlight}
 ]
}
